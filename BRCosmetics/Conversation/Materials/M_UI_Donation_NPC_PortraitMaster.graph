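# Translucent UI master material for an NPC portrait in donation conversations, compositing a parameterized portrait texture over a spotlight background.
# The background blends falloff and hotspot colors using a movable exponential radial gradient, color-dodged with itself at adjustable intensity.
# A tapered vertical UV gradient tints the portrait toward a depth-fade color, by default in its lower half.
# A procedural UV border mask with adjustable size, curve and hardness, scaled by FinalAlpha, provides soft-edged opacity.
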Material /BRCosmetics/Conversation/Materials/M_UI_Donation_NPC_PortraitMaster
graph {
 properties {blendMode=BLEND_Translucent, materialDomain=MD_UI, pixelDepthOffsetMode=PDOM_Legacy, usedShadingModels=Default Lit}
 nodes [
  n1 = Multiply(A=ScalarParameter() {defaultValue=16, group=Background Gradient, parameterName=Spotlight_Intensity}) {constB=0.01}
  n2 = AppendVector(
   A=Add(A=Constant() {r=0.5}, B=ScalarParameter() {parameterName=MoveGradient_X})
   B=Add(A=Constant() {r=0.5}, B=ScalarParameter() {parameterName=MoveGradient_Y})
  )
  n3 = MaterialFunctionCall(
   CenterPosition=@n2
   Radius=ScalarParameter() {defaultValue=0.5, parameterName=Radius}
   Density=ScalarParameter() {defaultValue=1.666, parameterName=Density}
  ) {materialFunction=&RadialGradientExponential}
  n4 = ScalarParameter() {defaultValue=0.5, parameterName=Radius}
  n5 = ScalarParameter() {defaultValue=1.666, parameterName=Density}
  n6 = MaterialFunctionCall!(
   Blend=Saturate(None=LinearInterpolate(B=@n1, Alpha=@n3.RadialGradientExponential))
   Base=Saturate(None=MaterialFunctionCall(CenterPosition=@n2, Radius=@n4, Density=@n5) {materialFunction=&RadialGradientExponential}.RadialGradientExponential)
  ) {materialFunction=/Engine/Functions/Engine_MaterialFunctions03/Blends/Blend_ColorDodge}
  n7 = TextureObjectParameter() {
   parameterName=NPC-Portrait
   texture=/BRCosmetics/UI/Foundation/Textures/Icons/Heroes/Athena/Soldier/T-Soldier-HID-A-199-Athena-Commando-M-SpaceChimp
  }
  n8 = TextureSample(Tex=@n7) {texture=&T-Soldier-HID-A-199-Athena-Commando-M-SpaceChimp}
  n9 = TextureCoordinate()
  n10 = ScalarParameter() {defaultValue=-0.5, parameterName=DepthFadeOffste}
  n11 = Multiply(A=ComponentMask(None=Add(A=@n9, B=@n10)) {g=true}, B=ScalarParameter() {defaultValue=1, parameterName=DepthFadeMultiply})
  n12 = Saturate(None=Power(Base=Saturate(None=@n11), Exp=ScalarParameter() {defaultValue=1.22, parameterName=DepthFadeTaper}))
  n13 = ScalarParameter() {defaultValue=0.75, parameterName=DepthFadeAmount}
  LinearInterpolate(
   A=LinearInterpolate(
    A=VectorParameter() {parameterName=FalloffColor, defaultValue=<rgba 0 0 1 1>}.RGB
    B=VectorParameter() {parameterName=HotspotColor, defaultValue=<rgba 0 1 0 1>}.RGB
    Alpha=Saturate(None=@n6.Result)
   )
   B=LinearInterpolate(A=@n8.RGB, B=VectorParameter() {parameterName=DepthFadeColor, defaultValue=<rgba 1 0 0 1>}.RGB, Alpha=Multiply(A=@n12, B=@n13))
   Alpha=@n8.A
  )
  n14 = TextureCoordinate()
  n15 = Abs(None=Divide(A=ConstantBiasScale(None=@n14) {bias=-0.5, scale=2}, B=ScalarParameter() {defaultValue=0.975, parameterName=EdgeSize}))
  n16 = OneMinus(None=ScalarParameter() {defaultValue=0.05, parameterName=EdgeCurve})
  n17 = Saturate(None=Divide(A=Subtract(A=@n15, B=@n16) {constB=0.25}, B=Subtract(B=@n16)))
  n18 = ComponentMask(None=@n17) {g=true}
  n19 = ComponentMask(None=@n17) {r=true}
  n20 = Multiply(A=LinearInterpolate(B=OneMinus(None=@n18), Alpha=OneMinus(None=@n19)), B=ScalarParameter() {defaultValue=2, parameterName=EdgeHardness})
  Saturate(None=Multiply(A=ScalarParameter() {defaultValue=1, parameterName=FinalAlpha}, B=Saturate(None=@n20)))
 ]
}
